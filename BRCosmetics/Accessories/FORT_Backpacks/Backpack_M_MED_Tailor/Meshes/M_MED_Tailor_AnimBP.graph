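# Animation Blueprint for the Tailor backpack cosmetic, procedurally animating its wheel, needle, spool, knob, pulley and lever bones with no animation assets.
# Each update, the EventGraph accumulates roll spins that wrap at a threshold and drives lever pitch and needle translation with sine waves on game time.
# The pawn's 2D speed scales every motion rate, keeping idle movement when stationary.
# The AnimGraph applies the values through chained additive bone-space Modify Bone nodes over the reference pose.
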AnimBlueprint /BRCosmetics/Accessories/FORT_Backpacks/Backpack_M_MED_Tailor/Meshes/M_MED_Tailor_AnimBP
target_skeleton=&/M_MED_Tailor_Skeleton
graphs [
 AnimGraph {nodes [
   n1 = "Transform (Modify) Bone - Bone: wheel"(
    ComponentPose=Local To Component(LocalPose=Local Space Ref Pose())
    Alpha=1
    Translation="0,0,0"
    Rotation=Get wheelRotate() -> (wheelRotate:Rotator)
    Scale="1,1,1"
   ) {
    rotationMode=BMM_Additive
    rotationSpace=BCS_BoneSpace
    translationMode=BMM_Additive
    translationSpace=BCS_BoneSpace
    alphaBoolBlend {blendOption=Linear}
    boneToModify {boneName=wheel}
   }
   n2 = "Transform (Modify) Bone - Bone: needle"(ComponentPose=@n1, Alpha=1, Translation=Get needleTranslate() -> (needleTranslate:Vector), Rotation="0,0,0", Scale="1,1,1") {
    rotationMode=BMM_Additive
    rotationSpace=BCS_BoneSpace
    translationMode=BMM_Additive
    translationSpace=BCS_BoneSpace
    alphaBoolBlend {blendOption=Linear}
    boneToModify {boneName=needle}
   }
   n3 = "Transform (Modify) Bone - Bone: spool"(ComponentPose=@n2, Alpha=1, Translation="0,0,0", Rotation=Get spoolRotate() -> (spoolRotate:Rotator), Scale="1,1,1") {
    rotationMode=BMM_Additive
    rotationSpace=BCS_BoneSpace
    translationMode=BMM_Additive
    translationSpace=BCS_BoneSpace
    alphaBoolBlend {blendOption=Linear}
    boneToModify {boneName=spool}
   }
   n4 = "Transform (Modify) Bone - Bone: knob"(ComponentPose=@n3, Alpha=1, Translation="0,0,0", Rotation=Get knobRotate() -> (knobRotate:Rotator), Scale="1,1,1") {
    rotationMode=BMM_Additive
    rotationSpace=BCS_BoneSpace
    translationMode=BMM_Additive
    translationSpace=BCS_BoneSpace
    alphaBoolBlend {blendOption=Linear}
    boneToModify {boneName=knob}
   }
   n5 = "Transform (Modify) Bone - Bone: pulley"(ComponentPose=@n4, Alpha=1, Translation="0,0,0", Rotation=Get pulleyRotate() -> (pulleyRotate:Rotator), Scale="1,1,1") {
    rotationMode=BMM_Additive
    rotationSpace=BCS_BoneSpace
    translationMode=BMM_Additive
    translationSpace=BCS_BoneSpace
    alphaBoolBlend {blendOption=Linear}
    boneToModify {boneName=pulley}
   }
   n6 = "Transform (Modify) Bone - Bone: lever"(ComponentPose=@n5, Alpha=1, Translation="0,0,0", Rotation=Get leverRotate() -> (leverRotate:Rotator), Scale="1,1,1") {
    rotationMode=BMM_Additive
    rotationSpace=BCS_BoneSpace
    translationMode=BMM_Additive
    translationSpace=BCS_BoneSpace
    alphaBoolBlend {blendOption=Linear}
    boneToModify {boneName=lever}
   }
   Output Pose(Result=Component To Local(ComponentPose=@n6)) {layerGroup=None}
  ]}
 EventGraph {nodes [
   n1 = Event BlueprintUpdateAnimation()
   > n2 = Branch(Condition="float >= float"(A=Get wheelSpin() -> (wheelSpin:double), B=720) {is=title K2Node_CallFunction})
   > n3 = Set wheelSpin(wheelSpin:double=float - float(A=Get wheelSpin(), B=720) {is=title K2Node_CallFunction}) -> (Output_Get:double)
   n4 = Get PawnSpeed2D()
   n5 = MapRangeClamped(Value=@n4, InRangeA=0, InRangeB=100, OutRangeA=1, OutRangeB=10)
   n6 = float + float(A=Get wheelSpin(), B=float * float!(A=float * float!(A=1, B=20) -> (ReturnValue), B=@n1.DeltaTimeX, C:double=@n5) -> (ReturnValue)) {is=title K2Node_CommutativeAssociativeBinaryOperator}
   Set wheelRotate(execute=[@n24.then, @n3.then], wheelRotate:Rotator=MakeRotator(Roll=@n6)) -> (Output_Get:Rotator)
   > n7 = Branch(Condition="float >= float"(A=Get pulleySpin() -> (pulleySpin:double), B=720))
   > n8 = Set pulleySpin(pulleySpin:double=float - float(A=Get pulleySpin(), B=720)) -> (Output_Get:double)
   n9 = float + float(A=Get pulleySpin(), B=float * float!(A=float * float!(A=1, B=20) -> (ReturnValue), B=@n1.DeltaTimeX, C:double=@n5) -> (ReturnValue))
   Set pulleyRotate(execute=[@n23.then, @n8.then], pulleyRotate:Rotator=MakeRotator(Roll=@n9)) -> (Output_Get:Rotator)
   > n10 = Branch(Condition="float >= float"(A=Get spoolSpin() -> (spoolSpin:double), B=720))
   > n11 = Set spoolSpin(spoolSpin:double=float - float(A=Get spoolSpin(), B=720)) -> (Output_Get:double)
   n12 = float + float(A=Get spoolSpin(), B=float * float!(A=float * float!(A=1, B=100) -> (ReturnValue), B=@n1.DeltaTimeX, C:double=@n5) -> (ReturnValue))
   Set spoolRotate(execute=[@n22.then, @n11.then], spoolRotate:Rotator=MakeRotator(Roll=@n12)) -> (Output_Get:Rotator)
   > n13 = Branch(Condition="float >= float"(A=Get knobSpin() -> (knobSpin:double), B=720))
   > n14 = Set knobSpin(knobSpin:double=float - float(A=Get knobSpin(), B=720)) -> (Output_Get:double)
   n15 = float + float(A=Get knobSpin(), B=float * float!(A=float * float!(A=1, B=60) -> (ReturnValue), B=@n1.DeltaTimeX, C:double=@n5) -> (ReturnValue))
   Set knobRotate(execute=[@n21.then, @n14.then], knobRotate:Rotator=MakeRotator(Roll=@n15)) -> (Output_Get:Rotator)
   n16 = GetGameTimeInSeconds()
   n17 = MapRangeClamped(Value=@n4, InRangeA=0, InRangeB=100, OutRangeA=1, OutRangeB=2)
   n18 = "Sin (Radians)"(A=float * float(A=float * float!(A=1, B=5, C:double=@n17) -> (ReturnValue), B=@n16))
   > Set leverRotate(leverRotate:Rotator=MakeRotator(Pitch=float * float!(A=@n18, B=10) -> (ReturnValue))) -> (Output_Get:Rotator)
   n19 = MapRangeClamped(Value=@n4, InRangeA=0, InRangeB=100, OutRangeA=1, OutRangeB=1.5)
   n20 = "Sin (Radians)"(A=float * float(A=float * float!(A=1, B=20, C:double=@n19) -> (ReturnValue), B=@n16))
   > Set needleTranslate(needleTranslate:Vector=MakeVector(X=float * float!(A=@n20, B=0.6) -> (ReturnValue), Y=0)) -> (Output_Get:Vector)
   n21 = Set knobSpin(execute=@n13.else, knobSpin=@n15)
   n22 = Set spoolSpin(execute=@n10.else, spoolSpin=@n12)
   n23 = Set pulleySpin(execute=@n7.else, pulleySpin=@n9)
   n24 = Set wheelSpin(execute=@n2.else, wheelSpin=@n6)
  ]}
]
variables [
 knobRotate {type=Rotator}
 knobSpin {type=number}
 leverRotate {type=Rotator}
 needleTranslate {type=Vector}
 pulleyRotate {type=Rotator}
 pulleySpin {type=number}
 spoolRotate {type=Rotator}
 spoolSpin {type=number}
 wheelRotate {type=Rotator}
 wheelSpin {type=number}
]
